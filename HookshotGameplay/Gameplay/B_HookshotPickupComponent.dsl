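; B_HookshotPickupComponent is a temporary component on a tossed FortPickup that manages its flight and landing.
; The server binds to projectile bounce, slows the pickup, sets a replicated bounce flag, then destroys itself after a grace delay.
; Pickups starting in water get a minimum upward velocity and are re-tossed toward the stored target on exit.
; The flag's OnRep plays landing audio from item data, the sound library or a default cue.
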(event EventBeginPlay
  ;[0.027 0.08333 0.03046] BeginPlay: On the server, bind to know when the projectile bounces.
  (|CallParentFunction)
  (Networking|SwitchHasAuthority
    (:Authority
      (Utilities|FlowControl|Sequence
        (:then_0
          (bind _asfort_pickup (Utilities|Casting|CastToFortPickup (Components|GetOwner))
            (:then
              (bind _movementcomponent (Variables|/Script/FortniteGame.FortPickup|GetMovementComponent _asfort_pickup)
                (:then
                  (EventDispatchers|BindEventtoOnProjectileBounce _movementcomponent (EventDispatchers|CreateEvent :!event Custom|HandlePickupBounce))
                  ;;[0.03243 0.2372 0.2396] If the pickup is in water, bind to know when it exits. This is because the toss velocity will need to be applied again once it exists the effects of the water so that it lands where it is supposed to.
                  (Utilities|IsValid (Actor|GetComponentbyClass _asfort_pickup "/Script/FortniteGame.FortWaterInteractionComponent")
                    (:Is%20Valid
                      (if (Water|IsinWater (Actor|GetComponentbyClass _asfort_pickup "/Script/FortniteGame.FortWaterInteractionComponent"))
                        (Water|BindEventtoOnExitWaterEvent (Actor|GetComponentbyClass _asfort_pickup "/Script/FortniteGame.FortWaterInteractionComponent") (EventDispatchers|CreateEvent :!event HandlePickupExitWater))
                        ;;[0.4762 0.7812 0.3116] Apply a minimum upward velocity to the pickup in the water so that it exits the water surface.
                        (bind (_velocity_x _velocity_y _velocity_z) (Math|Vector|BreakVector (Variables|/Script/Engine.MovementComponent|GetVelocity_X _movementcomponent)))
                        (Projectile|SetVelocity _movementcomponent (Math|Vector|MakeVector _velocity_x _velocity_y (Math|Float|Max(Float) _velocity_z 1e+03)))))
                    (:Is%20Not%20Valid
                      _)))
                (:else
                  _))
              (Components|DestroyComponent))
            (:CastFailed
              _)))
        (:then_1
          (Utilities|FlowControl|Delay (Variables|Settings|GetComponentLifetime)) ; Time after which it can be considered that bounce \nwas never called.\nKill the component after this.
        )))))

(event Custom|HandlePickupBounce
  (params (ImpactResult :type Hit%20Result%20Structure%20%28by%20ref%29) (ImpactVelocity :type Vector%20%28by%20ref%29))
  ;[0.04149 0.232 0.2344] HandlePickupBounce: Pickup has touched the ground after toss, reduce its velocity, and let the client know about it.
  (Networking|SwitchHasAuthority
    (:Authority
      (bind _asfort_pickup (Utilities|Casting|CastToFortPickup (Components|GetOwner)))
      (bind _movementcomponent (Variables|/Script/FortniteGame.FortPickup|GetMovementComponent _asfort_pickup)
        (:then
          (Projectile|SetVelocity _movementcomponent (/ (Variables|/Script/Engine.MovementComponent|GetVelocity _movementcomponent) (Variables|Settings|GetGrindMagnitude))) ; Slowing down every bounce.
          (Variables|Transient|SetbPickupBounced true) ; Responsible for playing loot hit the ground sound.
          (Utilities|FlowControl|Delay (Variables|Settings|GetPostBounceTimeBeforeKill)) ; Grace period to allow PickupBounced replication to complete
          (Components|DestroyComponent))
        (:else)))))

(event Custom|OnRep_bPickupBounced
  (bind _returnvalue (Components|GetOwner))
  (bind _returnvalue_2 (WorldSubsystems|GetSoundLibrarySubsystem))
  ;[0.1562 0.03548 0.1415] OnRep_bPickupBounced: Pickup has landed, play its landed sound.
  (Networking|CanExecuteCosmeticEvents
    (:True
      (bind _asfort_pickup (Utilities|Casting|CastToFortPickup (Components|GetOwner)))
      (bind _asfort_world_item_definition (Utilities|Casting|CastToFortWorldItemDefinition (Pickup|GetPrimaryItemDefinition _asfort_pickup)))
      ;;[0.02184 0.125 0.1175] Attempt to spawn the item's custom landed sound, if it exists.
      (bind _itemdata (Class|ItemDefinitionBase|GetItemData _asfort_world_item_definition)
        (:Found
          (if (Utilities|IsValidSoftObjectReference (.LandedSound (Utilities|Struct|BreakFortnitePickupItemData _itemdata)))
            (bind _object (Utilities|AsyncLoadAsset (.LandedSound (Utilities|Struct|BreakFortnitePickupItemData _itemdata)))
              (:Completed
                (bind _assound_base (Utilities|Casting|CastToSoundBase _object)
                  (:then
                    (Audio|PlaySoundatLocation _assound_base (Transformation|GetActorLocation (Components|GetOwner))))
                  (:CastFailed
                    _))))
            (else
              _))
          ;;[0.03548 0.175 0.2135] If custom sound is not specified, use the sound library subsystem to play the land sound.
          (Utilities|IsValid _returnvalue
            (:Is%20Valid
              (Utilities|IsValid _returnvalue_2
                (:Is%20Valid
                  (bind _returnvalue_3 (SoundLibrary|PlaySoundSimple _returnvalue_2 (Components|GetOwner) (Utilities|Struct|MakeSimpleSoundSettings "(TagName=\"SoundLibrary.Item.Drop.Impact\")" false false true false 1.0 1.0 0.0 0 0 "None" 0 "KeepWorldPosition" (Transformation|GetActorLocation _returnvalue)) :!class /Script/SoundLibrary.SoundLibrarySubsystem))
                  (if (not (.bSuccess (Utilities|Struct|BreakSoundLibraryPlayResult _returnvalue_3)))
                    _))
                (:Is%20Not%20Valid
                  _))
              ;;[0.1301 0.02691 0.1302] If sound library is not set up for the pickup, use the default pickup landed sound cue.
              (Audio|PlaySoundatLocation (Variables|Sound|GetDefaultPickupLandedSound) (Transformation|GetActorLocation (Components|GetOwner))))) ; Validate Owner after Async action again.
        )
        (:NotFound
          _)))))

(event EventEndPlay
  (params (EndPlayReason :type EEndPlayReason%20Enum :default "Destroyed"))
  ;[0.1302 0.07416 0.01715] EndPlay: Clean up.
  (Utilities|FlowControl|Sequence
    (:then_0
      (Networking|SwitchHasAuthority
        (:Authority
          (bind _asfort_pickup (Utilities|Casting|CastToFortPickup (Components|GetOwner)))
          (Utilities|FlowControl|Sequence
            (:then_0
              (bind _movementcomponent (Variables|/Script/FortniteGame.FortPickup|GetMovementComponent _asfort_pickup)
                (:then
                  (EventDispatchers|UnbindEventfromOnProjectileBounce _movementcomponent (EventDispatchers|CreateEvent :!event Custom|HandlePickupBounce)))
                (:else)))
            (:then_1
              (Utilities|IsValid (Actor|GetComponentbyClass _asfort_pickup "/Script/FortniteGame.FortWaterInteractionComponent")
                (:Is%20Valid
                  (Water|UnbindEventfromOnExitWaterEvent (Actor|GetComponentbyClass _asfort_pickup "/Script/FortniteGame.FortWaterInteractionComponent") (EventDispatchers|CreateEvent :!event HandlePickupExitWater)))))))))
    (:then_1
      (|CallParentFunction EndPlayReason))))

(event Custom|DestroyComponent
  ;[0.03583 0.1302 0.1185] DestroyComponent: Helps invoke K2_DestroyComponent() from a non-owning class.
  (Networking|SwitchHasAuthority
    (:Authority
      (Components|DestroyComponent))))

(fn HandlePickupExitWater
  (params (WaterBody :type Fort%20Water%20Body%20Actor%20Object%20Reference) (WaterInteractionComponent :type Fort%20Water%20Interaction%20Component%20Object%20Reference) (bIsLastBody :type Boolean))
  ;[0.0322 0.1287 0.1354] HandlePickupExitWater: The pickup has exited the water, reapply the toss velocity so that it lands at the desired location.
  (Networking|SwitchHasAuthority
    (:Authority
      (Utilities|IsValid WaterInteractionComponent
        (:Is%20Valid
          (Water|UnbindEventfromOnExitWaterEvent WaterInteractionComponent (EventDispatchers|CreateEvent :!event HandlePickupExitWater))
          (bind _asfort_pickup (Utilities|Casting|CastToFortPickup (Components|GetOwner)))
          (bind _outlaunchvelocity (Game|SuggestProjectileVelocityCustomArc (Transformation|GetActorLocation _asfort_pickup) (Variables|Transient|GetTossLocation) 0.0 (Variables|Transient|GetTossArc)))
          (Pickup|TossPickupinDirection _asfort_pickup _outlaunchvelocity 0 false false))))))
