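; Hookshot projectile that spawns a rope linked to the weapon, registers overlapped pickups and fishing-spot catches for pulling, and broadcasts its stop to start the hookshot.
; Client-suggested hits on movable components are validated against a short server history of timestamped locations.
; The server kills it beyond range from the player, while a replicated disable hides it and ends the rope.
; Cosmetics add water splashes; homing misses can straighten velocity away from the shooter.
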(event EventBeginPlay
  (bind _weapon (CallFunction|GetWeaponActor))
  ;[0.01187 0.08333 0.02386] BeginPlay: Spawn and attach the rope to the projectile.
  (|CallParentFunction)
  (Components|Capsule|SetCapsuleSize (Variables|Collision|GetCapsuleComponent) (ScalableFloat|GetValueatLevel (Variables|hotfix|GetHotfix_ProjectileRadius)) (ScalableFloat|GetValueatLevel (Variables|hotfix|GetHotfix_ProjectileRadius)) false)
  (Variables|Transient|SetbClientSuggestedImpactEnabled (ScalableFloat|GetValueasBool (Variables|hotfix|GetHotfix_ClientSuggestedImpactEnabled)))
  (Networking|SwitchHasAuthority
    (:Authority
      (Variables|Transient|SetProjectileSpawnGameTime (Fortnite|Sync|Time|GetServerWorldTimeSeconds))
      (Variables|Transient|SetServerLocationTrackingNumOfFrames (ScalableFloat|GetValueasInteger (Variables|hotfix|GetHotfix_ClientSuggestedImpactServerLocTrackingNumOfFames)))
      (if (and (Variables|Debug|GetDebugPreventCollisionOnServer) (Editor|InEditororPIE))
        (Collision|SetCollisionEnabled (Variables|Collision|GetCapsuleComponent) "NoCollision")
        (else
          _))
      (Utilities|IsValid _weapon
        (:Is%20Valid
          (bind _returnvalue (Game|SpawnActorfromClass (Variables|General|GetRopeClass) (Math|Conversions|ToTransform(Vector) (Transformation|GetActorLocation)) "Undefined" "MultiplyWithRoot" _weapon (CallFunction|GetPlayerPawn)))
          (bind _output_get (Variables|Transient|SetRope _returnvalue))
          (Utilities|IsValid _output_get
            (:Is%20Valid
              (Class|FortHookGunRope|SetWeaponandProjectile _output_get _weapon self))))))))

(event Custom|DisableProjectile
  (params (HitResult :type Hit%20Result%20Structure))
  ;[0.1302 0.07979 0.03246] DisableProjectile: Disable the projectile under various conditions.
  (Networking|SwitchHasAuthority
    (:Authority
      (Utilities|FlowControl|DoOnce false
        (:Completed
          (Variables|SetDisableProjectileRepInfo_bDisableProjectile_1_21290B17458EF5F6E3C57389CD3D05A7 (Utilities|Struct|MakeHookshotDisableProjectileRepInfo true HitResult)))))))

(event EventOnDisableForKill
  ;[0.07749 0.04234 0.1198] OnDisableForKill: Fired when the projectile  has traveled beyond its kill distance.
  (|CallParentFunction)
  (Networking|SwitchHasAuthority
    (:Authority
      (Default|CallOnProjectileKilled))))

(event EventTick
  (params (DeltaSeconds :type Float%20%28single-precision%29))
  (bind _playerpawn (CallFunction|GetPlayerPawn))
  (bind _servertimeandprojectilelocationpairs (Variables|Transient|GetServerTimeAndProjectileLocationPairs))
  ;[0.2123 0.2135 0.04221] Tick: Check to see if the projectile has gone too far from the player pawn; if it has, kill it.
  (|CallParentFunction DeltaSeconds)
  (Networking|SwitchHasAuthority
    (:Authority
      (if (not (Projectile|IsProjectileBeingKilled))
        (Utilities|IsValid _playerpawn
          (:Is%20Valid
            (if (> (Transformation|GetSquaredDistanceTo _playerpawn self) (Variables|Projectile|GetMaxSimulationDistanceSquaredOverride))
              (Delivery|Kill)
              (elif (Variables|Transient|GetbClientSuggestedImpactEnabled)
                ;;[0.01497 0.1198 0.1186] Track the world location over time. Only track the last 15 frames/ticks.
                (if (>= (Utilities|Array|Length _servertimeandprojectilelocationpairs) (Variables|Transient|GetServerLocationTrackingNumOfFrames))
                  (Utilities|Array|RemoveIndex _servertimeandprojectilelocationpairs)
                  (else
                    _))
                (Utilities|Array|Add _servertimeandprojectilelocationpairs (Utilities|Struct|MakeHookshotTimeandLocationPair (Fortnite|Sync|Time|GetServerWorldTimeSeconds) (Transformation|GetActorLocation)))
                (if (Variables|Debug|GetbDebugDrawProjectileLocationMap)
                  (for _array_element _servertimeandprojectilelocationpairs
                    (Rendering|Debug|DrawDebugSphere (.Location_5_E9EB30FD4E12F1EEF7CE6F8A8981C2B9 (Utilities|Struct|BreakHookshotTimeandLocationPair _array_element)) 5.0 12 "(R=1.000000,G=0.049568,B=0.741289,A=1.000000)" 0.0 1.0)))))))))))

(event OnEnterWaterEvent(WaterInteractionComponent)
  (params (WaterBody :type Fort%20Water%20Body%20Actor%20Object%20Reference) (WaterInteractionComponent :type Fort%20Water%20Interaction%20Component%20Object%20Reference) (bIsFirstBody :type Boolean))
  ;[0.2635 0.03734 0.2708] OnEnterWaterEvent: Play the splash FX (Visual and sound) on the water.
  (Networking|CanExecuteCosmeticEvents
    (:True
      (if bIsFirstBody
        (Utilities|IsValid WaterBody
          (:Is%20Valid
            (bind (_outhit _returnvalue) (Collision|LineTraceForObjects (Movement|GetPreviousLocation) (Transformation|GetActorLocation) (Utilities|Array|MakeArray "ObjectTypeQuery13")))
            (bind _outwatersurfaceinfo (WaterBody|GetWaterSurfaceInfoStruct WaterBody (select _returnvalue (.ImpactPoint (Collision|BreakHitResult _outhit)) (Transformation|GetActorLocation)) true true false))
            (Utilities|FlowControl|Sequence
              (:then_0
                (bind _ns_watersplash (Variables|Water|GetNS_WaterSplash)
                  (:then
                    (bind _object (Utilities|AsyncLoadAsset _ns_watersplash)
                      (:Completed
                        (bind _asfxsystem_asset (Utilities|Casting|CastToFXSystemAsset _object))
                        (bind _returnvalue_2 (Effects|Components|ParticleSystem|SpawnFXAtLocation _asfxsystem_asset (Utilities|Struct|BreakWaterSurfaceInfo _outwatersurfaceinfo) (Math|Rotator|MakeRotfromX (.WaterPlaneNormal (Utilities|Struct|BreakWaterSurfaceInfo _outwatersurfaceinfo)))))
                        (Utilities|IsValid _returnvalue_2
                          (:Is%20Valid
                            (Effects|Components|ParticleSystem|SetFloatParameter _returnvalue_2 "WaterDepth" (.WaterDepth (Utilities|Struct|BreakWaterSurfaceInfo _outwatersurfaceinfo)))
                            (Effects|Components|ParticleSystem|SetFloatParameter _returnvalue_2 "WaterBodyIndex" (Math|Conversions|ToFloat(Integer) (.WaterBodyIdx (Utilities|Struct|BreakWaterSurfaceInfo _outwatersurfaceinfo))))
                            (bind _object_2 (Utilities|AsyncLoadAsset (Variables|Water|GetWaterSplashSoundCue))
                              (:Completed
                                (bind _assound_base (Utilities|Casting|CastToSoundBase _object_2))
                                (Audio|PlaySoundatLocation _assound_base (Utilities|Struct|BreakWaterSurfaceInfo _outwatersurfaceinfo)))))))))
                  (:else)))
              (:then_1
                ;;[1 0.2488 0.2289] Water plays fluid force impulse material animation based on the passed in values.
                (bind _asfort_water_body_bp (Utilities|Casting|CastToFortWaterBodyBP WaterBody))
                (bind _fluidsimbp (Variables|/Game/Athena/Apollo/Environments/Landscape/Water/FortWaterBodyBP.FortWaterBodyBP_C|GetFluidSimBP _asfort_water_body_bp)
                  (:then
                    (Class|BPFluidSim01|ApplyFluidForceImpulse _fluidsimbp (Utilities|Struct|MakeFluidForceImpulse (Utilities|Struct|BreakWaterSurfaceInfo _outwatersurfaceinfo) 1e+02 -1e+02 "/FortWater/FluidSimulation/Materials/Forces/M_Fluid_Sim_Force_Impulse.M_Fluid_Sim_Force_Impulse" 0.0 0.0 (Variables|Water|GetTimedWaterDropSplashes))))
                  (:else))))))))))

(event ProjectileMovement|EventOnHomingPassedTarget
  (bind _returnvalue (Actor|GetOwner))
  ;[0.5952 1 0.4518] OnHomingPassedTarget: If the target has been passed, stop homing and straighten out velocity
  (Projectile|SetHomingStyle 0 0 "None" false :!class /Script/FortniteGame.FortProjectileMovementInterface)
  (if (ScalableFloat|GetValueasBool (Variables|Default|GetHotfix_CorrectVelocityAfterHomingMiss))
    (bind _projectilemovementcomponent (Variables|Movement|GetProjectileMovementComponent)
      (:then
        (Utilities|IsValid _returnvalue
          (:Is%20Valid
            (Projectile|SetVelocity _projectilemovementcomponent (* (Math|Vector|Normalize (- (Transformation|GetActorLocation) (Transformation|GetActorLocation _returnvalue))) (Math|Vector|VectorLength (Variables|/Script/Engine.MovementComponent|GetVelocity _projectilemovementcomponent)))) ; Set velocity to be directly away from shooter \ninstead of potentially at an angle
          )))
      (:else))))

(event OnComponentBeginOverlap(CapsuleComponent)
  (params (OverlappedComponent :type Primitive%20Component%20Object%20Reference) (OtherActor :type Actor%20Object%20Reference) (OtherComp :type Primitive%20Component%20Object%20Reference) (OtherBodyIndex :type Integer) (bFromSweep :type Boolean) (SweepResult :type Hit%20Result%20Structure%20%28by%20ref%29))
  ;[0.02888 0.1875 0.1613] OnComponentBeginOverlap: Handle overlap pickups and floppers.
  (Networking|SwitchHasAuthority
    (:Authority
      (CallFunction|HandleOverlapPickupActor OtherActor)
      (CallFunction|HandleOverlapFlopper OtherActor))))

(event EventOnStop
  (params (Hit :type Hit%20Result%20Structure%20%28by%20ref%29))
  ;[0.01171 0.15 0.127 0.5] Broadcast projectile stop to start hookshot
  (|CallParentFunction Hit)
  (Default|CallOnProjectileStopped (Utilities|Array|MakeArray Hit)))

(event EventEndPlay
  (params (EndPlayReason :type EEndPlayReason%20Enum :default "Destroyed"))
  ;[0.125 0.08338 0.0293] EndPlay: Clean up the delegate binds.
  (Utilities|FlowControl|Sequence
    (:then_0
      (Networking|SwitchHasAuthority
        (:Authority
          (for _array_element (Variables|Transient|GetOverlappedFlopperSpawns)
            (Utilities|IsValid _array_element
              (:Is%20Valid
                (Default|UnbindEventfromOnItemSpawned _array_element (EventDispatchers|CreateEvent :!event HandleFlopperItemSpawned))))))))
    (:then_1
      (|CallParentFunction EndPlayReason))))

(event EventOnTouched
  (params (OtherActor :type Actor%20Object%20Reference) (OtherComp :type Primitive%20Component%20Object%20Reference) (HitResult :type Hit%20Result%20Structure%20%28by%20ref%29) (bIsOverlap :type Boolean))
  ;[0.01232 0.194 0.1979] OnTouched: On the client, when the projectile touches an actor - send an RPC to the server with this information.
  ; The RPC is only called if the projectile is still moving on the client.
  (Networking|CanExecuteCosmeticEvents
    (:True
      (if (Variables|Transient|GetbClientSuggestedImpactEnabled)
        (bind _breturnvalue (ServerRPC|IsProjectileSimulating))
        (if _breturnvalue
          (bind _returnvalue (ServerRPC|VerifyHitResultforServerHitRPC HitResult))
          (if _returnvalue
            (CallFunction|NotifyServerHit HitResult (Fortnite|Sync|Time|GetServerWorldTimeSeconds))))))))

(event Custom|NotifyServerHit
  (params (HitResult :type Hit%20Result%20Structure) (ServerWorldTimeSeconds :type Float%20%28double-precision%29))
  (bind (_bblockinghit _binitialoverlap _time _distance _location _impactpoint _normal _impactnormal _physmat _hitactor _hitcomponent) (Collision|BreakHitResult HitResult))
  (bind _hotfix_clientsuggestedimpactvaliddistance (Variables|hotfix|GetHotfix_ClientSuggestedImpactValidDistance))
  (bind _returnvalue (ScalableFloat|GetValueatLevel _hotfix_clientsuggestedimpactvaliddistance))
  (bind (_bblockinghit_2 _binitialoverlap_2 _time_2 _distance_2 _location_2 _impactpoint_2) (Collision|BreakHitResult HitResult))
  (bind (_bblockinghit_3 _binitialoverlap_3 _time_3 _distance_3 _location_3 _impactpoint_3 _normal_2 _impactnormal_2 _physmat_2 _hitactor_2 _hitcomponent_2 _hitbonename) (Collision|BreakHitResult HitResult))
  ;[0.07292 0.01408 0.06386] NotifyServerHit: The client is letting the server know that the projectile has hit an actor for the client. If the server is still simulating the projectile, check to see how far the client's impact point on the hit result is compared to the projectile location on the server. If this test passes, stop simulating the projectile with the client's information.
  (Networking|SwitchHasAuthority
    (:Authority
      (if (Variables|Transient|GetbClientSuggestedImpactEnabled)
        (bind _breturnvalue (ServerRPC|IsProjectileSimulating))
        (if _breturnvalue
          (bind _returnvalue_2 (ServerRPC|VerifyHitResultforServerHitRPC HitResult))
          (if _returnvalue_2
            (if _bblockinghit
              (Fortnite|FortShippingLog self (Utilities|String|Append "NotifyServerHit: Notifying server hit actor [" (Utilities|GetDisplayName _hitactor) "] ] on component [" (Utilities|GetDisplayName _hitcomponent) "] at server time [" (Utilities|String|ToString(Float) ServerWorldTimeSeconds) "]."))
              (bind _location_4 (ServerRPC|GetProjectileLocationatServerTime ServerWorldTimeSeconds))
              (if (> (.ReturnValue (Collision|GetClosestPointonCollision _hitcomponent_2 _location_4 _hitbonename)) 0)
                (if (< (.ReturnValue (Collision|GetClosestPointonCollision _hitcomponent_2 _location_4 _hitbonename)) _returnvalue)
                  (Game|Components|ProjectileMovement|StopSimulating (Variables|Movement|GetProjectileMovementComponent) HitResult)
                  (Transformation|SetActorLocation _impactpoint_2)
                  (Fortnite|FortShippingLog self (Utilities|String|Append "NotifyServerHit: Accepting client impact suggestion. Calculated Distance = " (Utilities|String|ToString(Float) (.ReturnValue (Collision|GetClosestPointonCollision _hitcomponent_2 _location_4 _hitbonename)))))
                  (else
                    (Fortnite|FortShippingLog self (Utilities|String|Append "NotifyServerHit: Ignoring as distance check failed." "Valid Max Distance = " (Utilities|String|ToString(Float) _returnvalue) " ; Calculated Distance = " (Utilities|String|ToString(Float) (.ReturnValue (Collision|GetClosestPointonCollision _hitcomponent_2 _location_4 _hitbonename))))))))))
          (else
            (Fortnite|FortShippingLog self "NotifyServerHit: Ignoring as projectile is no longer simulating.")))))))

(event Custom|OnRep_DisableProjectileRepInfo
  (bind (_disableprojectilerepinfo_bdisableprojectile_1_21290b17458ef5f6e3c57389cd3d05a7 _disableprojectilerepinfo_hitresult_6_8509609a495e0c5dd1f2af8e84db71c2) (Utilities|Struct|BreakHookshotDisableProjectileRepInfo (Variables|GetDisableProjectileRepInfo_bDisableProjectile_1_21290B17458EF5F6E3C57389CD3D05A7)))
  (bind (_bblockinghit _binitialoverlap _time _distance _location _impactpoint _normal _impactnormal _physmat _hitactor) (Collision|BreakHitResult _disableprojectilerepinfo_hitresult_6_8509609a495e0c5dd1f2af8e84db71c2))
  ;[0.02027 0.1263 0.1302] OnRep_bDisableProjectileRepInfo: Hides the actor, disables the actor collision, and let's the rope know that it is time to end its behavior.
  ; Called on both server and client.
  (if _disableprojectilerepinfo_bdisableprojectile_1_21290b17458ef5f6e3c57389cd3d05a7
    (Rendering|SetActorHiddenInGame true)
    (Collision|SetActorEnableCollision false)
    (Utilities|FlowControl|Sequence
      (:then_0
        (bind _rope (Variables|Transient|GetRope)
          (:then
            (Variables|/Script/FortniteGame.FortHookGunRope|SetbProjectileGone true _rope)
            (Class|FortHookGunRope|SetAttachedActor _rope _hitactor))
          (:else
            _))
        (bind _audio (Variables|Default|GetAudio)
          (:then
            (Audio|Components|Audio|FadeOut _audio 0.5 0.0 "Linear" :!class /Script/Engine.AudioComponent))
          (:else)))
      (:then_1
        (Networking|SwitchHasAuthority
          (:Authority
            (Fortnite|FortShippingLog self (Utilities|String|Append "DisableProjectile: Projectile lifetime: " (Utilities|String|ToString(Float) (- (Fortnite|Sync|Time|GetServerWorldTimeSeconds) (Variables|Transient|GetProjectileSpawnGameTime)))))))))))

(event Custom|OnRep_Rope)

(fn GetCurrentRopeLength
  ;[0.1654 0.06404 0.2188] GetCurrentRopeLength: Returns the length of the hookshot rope.
  (Utilities|FlowControl|Sequence
    (:then_0
      (bind _rope (CallFunction|GetRope))
      (Utilities|IsValid _rope
        (:Is%20Valid
          (bind _length (Class|BHookshotRope|GetRopeLength _rope))
          (return _length))))
    (:then_1
      (return 0.0))))

(fn GetPlayerPawn
  (return (Utilities|Casting|CastToFortPlayerPawn (Game|GetInstigator))))

(fn GetProjectileLocationAtServerTime
  (params (ServerWorldTimeSeconds :type Float%20%28double-precision%29))
  (bind (_servertime_2_6d2a4e334755c7c11e58aa8539c65bbb _location_5_e9eb30fd4e12f1eef7ce6f8a8981c2b9) (Utilities|Struct|BreakHookshotTimeandLocationPair (Utilities|Array|Get(aref) (Variables|Transient|GetServerTimeAndProjectileLocationPairs) (LocalVariables|GetClosestKeyIndex))))
  (bind _servertimeandprojectilelocationpairs (Variables|Transient|GetServerTimeAndProjectileLocationPairs))
  (bind _closestkeyindex (LocalVariables|GetClosestKeyIndex))
  ;[0.01298 0.1531 0.1562] Returns the location of the projectile near the provided server world time seconds.
  (Utilities|FlowControl|Sequence
    (:then_0
      (Networking|SwitchHasAuthority
        (:Authority
          (for (_array_element _array_index) _servertimeandprojectilelocationpairs
            ;;[0 1 0.955] Loop Body: Find the closest time (since the array is ordered by time we could optimize this with a binary search, but right now that isn't worth the risk)
            (if (< (Math|Float|Absolute(Float) (- (Utilities|Struct|BreakHookshotTimeandLocationPair _array_element) ServerWorldTimeSeconds)) (LocalVariables|GetClosestDifferenceValue))
              (LocalVariables|SetClosestDifferenceValue (Math|Float|Absolute(Float) (- (Utilities|Struct|BreakHookshotTimeandLocationPair _array_element) ServerWorldTimeSeconds)))
              (LocalVariables|SetClosestKeyIndex _array_index)))
          ;;[1 0.4427 0.4427] Return the location at the time closest to Server World Time Seconds
          (if (Utilities|Array|IsValidIndex _servertimeandprojectilelocationpairs _closestkeyindex)
            (return _location_5_e9eb30fd4e12f1eef7ce6f8a8981c2b9)))))
    (:then_1
      ;[1 0.4792 0.4792] Failed to find a time, return this actor's current location.
      (return (Transformation|GetActorLocation)))))

(fn GetRope
  ;[0.02636 0.1861 0.1979] Return the rope actor.
  (return (Variables|Transient|GetRope)))

(fn GetWeaponActor
  (Utilities|FlowControl|Sequence
    (:then_0
      (bind _asfort_player_pawn (Utilities|Casting|CastToFortPlayerPawn (Actor|GetOwner)))
      (return (Weapon|GetCurrentWeapon _asfort_player_pawn)))
    (:then_1
      (return 0))))

(fn HandleFlopperItemSpawned
  (params (Pickup :type Fort%20Pickup%20Object%20Reference))
  (Networking|SwitchHasAuthority
    (:Authority
      (CallFunction|HandleOverlapPickupActor Pickup))))

(fn HandleOverlapFlopper
  (params (Actor :type Actor%20Object%20Reference))
  ;[0.02173 0.1667 0.1517] HandleOverlapFlopper: When a fishing pond is overlapped, have it spawn an item that gets pulled.
  (Networking|SwitchHasAuthority
    (:Authority
      (if (ScalableFloat|GetValueasBool (Variables|hotfix|GetHotfix_FishingEnable))
        (bind _asbga_athena_flopper_spawn_parent (Utilities|Casting|CastToBGA_Athena_FlopperSpawn_Parent Actor))
        (bind _returnvalue (Utilities|Array|AddUnique (Variables|Transient|GetOverlappedFlopperSpawns) _asbga_athena_flopper_spawn_parent))
        (if (>= _returnvalue 0)
          (Default|BindEventtoOnItemSpawned _asbga_athena_flopper_spawn_parent (EventDispatchers|CreateEvent :!event HandleFlopperItemSpawned))
          (else
            _))
        (Class|BGAAthenaFlopperSpawnParent|SpawnItem :self _asbga_athena_flopper_spawn_parent :Target (Game|GetInstigator self) :Hooked true :SpawnLoc (Transformation|GetActorLocation) :Player (CallFunction|GetPlayerPawn) :ItemSpawnDelay 0.0 :CaughtWithFishingRod false :CaughtWithHappyGhost true :ItemUsedToFish (CallFunction|GetWeaponActor))))))

(fn HandleOverlapPickupActor
  (params (OverlappedActor :type Actor%20Object%20Reference))
  (bind _playerpawn (CallFunction|GetPlayerPawn))
  ;[0.07555 0.2083 0.1833] HandleOverlapPickupActor: Attach valid overlapped pickup to the rope so that it can be pulled back at the end of the hookshot ability.
  (Networking|SwitchHasAuthority
    (:Authority
      (if (ScalableFloat|GetValueasBool (Variables|hotfix|GetHotfix_PickupAttachEnable))
        (bind _asfort_pickup (Utilities|Casting|CastToFortPickup OverlappedActor))
        (Utilities|IsValid _playerpawn
          (:Is%20Valid
            (if (Math|Boolean|ANDBoolean (>= (Transformation|GetSquaredDistanceTo _playerpawn _asfort_pickup) (ScalableFloat|GetValueatLevel (Variables|hotfix|GetHotfix_PickupRegisterMinDistSqFromPlayer))) (<= (Transformation|GetSquaredDistanceTo _asfort_pickup self) (ScalableFloat|GetValueatLevel (Variables|hotfix|GetHotfix_PickupRegisterMinDistSqFromPrj))) (Utilities|NotEqual(Object) (Actor|GetAttachParentActor _asfort_pickup) self) (not (Pickup|PickedUp _asfort_pickup)) (not (Game|IsActorBeingDestroyed _asfort_pickup)))
              (bind _rope (Variables|Transient|GetRope)
                (:then
                  (Pickups|RegisterPickuptoPull _rope _asfort_pickup))
                (:else)))))))))

(fn IsProjectileSimulating
  ;[0.03038 0.1835 0.224] Returns if the projectile is simulating.
  (bind _projectilemovementcomponent (Variables|Movement|GetProjectileMovementComponent)
    (:then
      (Variables|/Script/Engine.MovementComponent|GetUpdatedComponent _projectilemovementcomponent
        (:then
          (if (Components|Activation|IsActive _projectilemovementcomponent)
            (return true)
            (else
              _)))
        (:else
          _))
      (return false))
    (:else
      _)))

(fn ConstructionScript)

(fn VerifyHitResultForServerHitRPC
  (params (HitResult :type Hit%20Result%20Structure))
  (bind (_bblockinghit _binitialoverlap _time _distance _location _impactpoint _normal _impactnormal _physmat _hitactor _hitcomponent) (Collision|BreakHitResult HitResult))
  (Utilities|FlowControl|Sequence
    (:then_0
      (if _bblockinghit
        (Utilities|IsValid _hitactor
          (:Is%20Valid
            (Utilities|IsValid _hitcomponent
              (:Is%20Valid
                (if (Utilities|Enum|Equal(Enum) (Variables|/Script/Engine.SceneComponent|GetMobility _hitcomponent) "Movable")
                  (return true))))))))
    (:then_1
      (return false))))
